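# Idle Niagara effect for the Dyed Duelist pickaxe, showing two looping, differently sized flames with glow on the user-supplied skeletal mesh.
# Flame001 samples a bone or socket per partition every update, adding a waveform bob and separate core parameters for two sprite renderers.
# Flame_Light adds sprite and light renderers, with alpha driven by flicker, emitter age and time of day.
# The Flame emitter, which samples only at spawn, is disabled.
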NiagaraSystem /BRCosmetics/Weapons/FORT_Melee/Pickaxe_DyedDuelist/FX/NS_Pickaxe_DyedDuelist_Idle
dependencies {usedDataInterfaces [&NiagaraDataInterfaceCurve], usedDynamicInputs [], usedModules [&SystemState]}
emitters {
 Flame {
  bEnabled=false
  emitterUpdateScript [
   &EmitterState(Life Cycle Mode=::NewEnumerator1, Loop Behavior=::NewEnumerator1, Loop Duration Mode=::NewEnumerator1, UseLoopDelay=false)
   &SpawnBurst_Instantaneous(Spawn Count=2, Use Spawn Probability=false, UseLoopCountLimit=false)
  ]
  particleSpawnScript [
   &PartitionParticles()
   &InitializeParticle(
    Color=<rgba 0.5218 0.0522 1 1>
    Lifetime=2
    Sprite Rotation Angle=&Lerp_Float(A=0, B=180, Alpha=&MakeFloatFromInt(Int=$Particles.PartitionParticles.Partition))
    Sprite Rotation Mode=::NewEnumerator0
    Sprite Size Mode=::NewEnumerator3
    Uniform Sprite Size=&Lerp_Float(A=60, B=25, Alpha=&MakeFloatFromInt(Int=$Particles.PartitionParticles.Partition))
    UsePositionOffset=false
    Write Material Random=true
   )
   &SkeletalMeshLocation(
    Bone / Socket Index=$Particles.PartitionParticles.Partition
    MaskSpawnGroup=false
    Mesh Sampling Type=::NewEnumerator5
    Skeletal Mesh=$User.Skeletal Mesh
    Socket Sampling Mode=::NewEnumerator5
   )
   &SpriteFacingAndAlignment(Sprite Alignment=<0 0 1>, Sprite Facing=<0 1 0>, WriteAlignment=true, WriteFacing=false)
   /SetVariables(
    Particles.FlameCoreColor:LinearColor=<rgba 0.3333 2 1.8508 1>
    Particles.FlameCorePosition:NiagaraPosition=&AddVectorToPosition(
     Position=$Particles.Position
     Vector=&Lerp_Vector(A=<0 0 4>, B=<0 0 0>, Alpha=&MakeFloatFromInt(Int=$Particles.PartitionParticles.Partition))
    )
    Particles.FlameCoreSize:Vector2f=&Vector2DFromFloat(Value=&Lerp_Float(A=35, B=10, Alpha=&MakeFloatFromInt(Int=$Particles.PartitionParticles.Partition)))
   ) {isSetParameters=true, module=SetVariables_#1}
  ]
  particleUpdateScript [
   &ParticleState(
    Kill Particles When Lifetime Has Elapsed=false
    Let Infinitely Lived Particles Die When Emitter Deactivates=true
    Loop Particles Lifetime=true
   )
   /SetVariables(Particles.Position=&AddVectorToPosition(
     Position=$Particles.Initial.Position
     Vector=&MakeVector(
      X=0
      Y=0
      Z=&Waveform(
       "Waveform [1]"=::NewEnumerator0
       "[1] Phase Scale"=&RandomRangeFloat(Minimum=1, Maximum=2)
       "[1] Frequency"=1
       "[1] Amplitude Min/Max"=<0 1>
       "[1] Scale"=2
      )
     )
    )) {isSetParameters=true, module=SetVariables_#2}
   &SolveForcesAndVelocity(Clamp Velocity=false, Limit Acceleration=false)
   &DynamicMaterialParameters(
    Index 0 Param 3=&FloatFromCurve(
     FloatCurve=DataInterface() {propertyValues {Curve {keys [{interpMode=RCIM_Cubic, tangentMode=RCTM_User, time=0, value=0}, {interpMode=RCIM_Cubic, tangentMode=RCTM_User, time=1, value=2}]}}}
     CurveIndex=$Particles.Timeline.NormalizedTime
    )
    Module.Index0Param0Write=false
    Module.Index0Param1Write=false
    Module.Index0Param2Write=true
    Module.Index0Param3Write=false
   )
  ]
  rendererClasses [&NiagaraSpriteRendererProperties]
  renderers [
   {rendererClass=&NiagaraSpriteRendererProperties}
   {rendererClass=&NiagaraSpriteRendererProperties, rendererIndex=1}
  ]
 }
 Flame001 {
  emitterUpdateScript [
   &EmitterState(Life Cycle Mode=::NewEnumerator1, Loop Behavior=::NewEnumerator1, Loop Duration Mode=::NewEnumerator1, UseLoopDelay=false)
   &SpawnBurst_Instantaneous(Spawn Count=2, Use Spawn Probability=false, UseLoopCountLimit=false)
  ]
  particleSpawnScript [
   &PartitionParticles()
   &InitializeParticle(
    Color=<rgba 0.5218 0.0522 1 1>
    Lifetime=2
    Sprite Rotation Angle=&Lerp_Float(A=0, B=180, Alpha=&MakeFloatFromInt(Int=$Particles.PartitionParticles.Partition))
    Sprite Rotation Mode=::NewEnumerator0
    Sprite Size Mode=::NewEnumerator3
    Uniform Sprite Size=&Lerp_Float(A=60, B=25, Alpha=&MakeFloatFromInt(Int=$Particles.PartitionParticles.Partition))
    UsePositionOffset=false
    Write Material Random=true
   )
   &SpriteFacingAndAlignment(Sprite Alignment=<0 0 1>, Sprite Facing=<0 1 0>, WriteAlignment=true, WriteFacing=false)
  ]
  particleUpdateScript [
   &ParticleState(
    Kill Particles When Lifetime Has Elapsed=false
    Let Infinitely Lived Particles Die When Emitter Deactivates=true
    Loop Particles Lifetime=true
   )
   &SkeletalMeshLocation(
    Bone / Socket Index=$Particles.PartitionParticles.Partition
    MaskSpawnGroup=false
    Mesh Sampling Type=::NewEnumerator5
    Skeletal Mesh=$User.Skeletal Mesh
    Socket Sampling Mode=::NewEnumerator5
   )
   /SetVariables(
    Particles.FlameCoreColor:LinearColor=<rgba 0.3333 2 1.8508 1>
    Particles.FlameCorePosition:NiagaraPosition=&AddVectorToPosition(
     Position=$Particles.Position
     Vector=&Lerp_Vector(A=<0 0 4>, B=<0 0 0>, Alpha=&MakeFloatFromInt(Int=$Particles.PartitionParticles.Partition))
    )
    Particles.FlameCoreSize:Vector2f=&Vector2DFromFloat(Value=&Lerp_Float(A=35, B=10, Alpha=&MakeFloatFromInt(Int=$Particles.PartitionParticles.Partition)))
   ) {isSetParameters=true, module=SetVariables_#1}
   /SetVariables(Particles.Position=&AddVectorToPosition(
     Position=$Particles.SkeletalMeshLocation.SampledPosition
     Vector=&MakeVector(
      X=0
      Y=0
      Z=&Waveform(
       "Waveform [1]"=::NewEnumerator0
       "[1] Phase Scale"=&RandomRangeFloat(Minimum=1, Maximum=2)
       "[1] Frequency"=1
       "[1] Amplitude Min/Max"=<0 1>
       "[1] Scale"=2
      )
     )
    )) {isSetParameters=true, module=SetVariables_#2}
   &SolveForcesAndVelocity(Clamp Velocity=false, Limit Acceleration=false)
   &DynamicMaterialParameters(
    Index 0 Param 3=&FloatFromCurve(
     FloatCurve=DataInterface() {propertyValues {Curve {keys [{interpMode=RCIM_Cubic, tangentMode=RCTM_User, time=0, value=0}, {interpMode=RCIM_Cubic, tangentMode=RCTM_User, time=1, value=2}]}}}
     CurveIndex=$Particles.Timeline.NormalizedTime
    )
    Module.Index0Param0Write=false
    Module.Index0Param1Write=false
    Module.Index0Param2Write=true
    Module.Index0Param3Write=false
   )
  ]
  rendererClasses [&NiagaraSpriteRendererProperties]
  renderers [
   {rendererClass=&NiagaraSpriteRendererProperties}
   {rendererClass=&NiagaraSpriteRendererProperties, rendererIndex=1}
  ]
 }
 Flame_Light {
  emitterUpdateScript [
   &EmitterState(Life Cycle Mode=::NewEnumerator1, Loop Behavior=::NewEnumerator1, Loop Duration Mode=::NewEnumerator1, UseLoopDelay=false)
   &SpawnBurst_Instantaneous(Spawn Count=2, Use Spawn Probability=false, UseLoopCountLimit=false)
  ]
  particleSpawnScript [
   &PartitionParticles()
   &InitializeParticle(
    Color=<rgba 1.8182 0.2628 3 0.1>
    Lifetime=2
    Sprite Rotation Angle=&Lerp_Float(A=0, B=180, Alpha=&MakeFloatFromInt(Int=$Particles.PartitionParticles.Partition))
    Sprite Rotation Mode=::NewEnumerator0
    Sprite Size=&Lerp_Vector2(A=<40 30>, B=<20 15>, Alpha=&MakeFloatFromInt(Int=$Particles.PartitionParticles.Partition))
    Sprite Size Mode=::NewEnumerator4
    UsePositionOffset=false
    Write Material Random=true
   )
  ]
  particleUpdateScript [
   &ParticleState(
    Kill Particles When Lifetime Has Elapsed=false
    Let Infinitely Lived Particles Die When Emitter Deactivates=true
    Loop Particles Lifetime=true
   )
   &SkeletalMeshLocation(
    Bone / Socket Index=$Particles.PartitionParticles.Partition
    MaskSpawnGroup=false
    Mesh Sampling Type=::NewEnumerator5
    Sampled Position Offset=&Lerp_Vector(A=<0 0 10>, B=<0 0 -4>, Alpha=&MakeFloatFromInt(Int=$Particles.PartitionParticles.Partition))
    Skeletal Mesh=$User.Skeletal Mesh
    Socket Sampling Mode=::NewEnumerator5
   )
   &SolveForcesAndVelocity(Clamp Velocity=false, Limit Acceleration=false)
   /SetVariables(Particles.Jitter:NiagaraFloat=&Waveform("Waveform [1]"=::NewEnumerator12, "[1] Phase Scale"=2, "[1] Frequency"=12, "[1] Amplitude Min/Max"=<0.85 1.3>)) {isSetParameters=true}
   /SetVariables(Particles.LightColor=<rgba 2.1899 0.4484 3 2>) {isSetParameters=true}
   &ScaleColor(
    Scale Alpha=&FloatFromCurve(
     FloatCurve=DataInterface() {propertyValues {Curve {keys [
         {arriveTangent=104.8297, interpMode=RCIM_Cubic, leaveTangent=104.8297, tangentMode=RCTM_User, time=0, value=0}
         {arriveTangent=-8.95, interpMode=RCIM_Cubic, leaveTangent=-8.95, tangentMode=RCTM_User, time=0.1, value=3}
         <at 0.5 1>
        ]}}}
     CurveIndex=$Emitter.Age
     Scale Curve=$Particles.Jitter
    )
    ScaleA=true
    ScaleRGB=false
   )
   &ScaleColor(
    Scale Alpha=&FloatFromCurve(
     FloatCurve=DataInterface() {propertyValues {Curve {keys [
         {interpMode=RCIM_Cubic, tangentMode=RCTM_User, time=0, value=0.9}
         {interpMode=RCIM_Cubic, tangentMode=RCTM_User, time=0.5, value=1.1}
         {tangentMode=RCTM_User, time=1, value=0.9}
        ]}}}
     CurveIndex=Linked() {linkedVariable {name=NPC.FortniteNPC.TimeOfDay_0-1, type=NiagaraFloat}}
    )
    ScaleA=true
    ScaleRGB=false
   ) {module=ScaleColor001}
  ]
  rendererClasses [&NiagaraSpriteRendererProperties, &NiagaraLightRendererProperties]
  renderers [
   {rendererClass=&NiagaraSpriteRendererProperties}
   {rendererClass=&NiagaraLightRendererProperties, rendererIndex=1}
  ]
 }
}
summary {
 systemName=NS_Pickaxe_DyedDuelist_Idle
 emitters [
  {bEnabled=false, emitterName=Flame, rendererClasses [&NiagaraSpriteRendererProperties]}
  {emitterName=Flame_Light, rendererClasses [&NiagaraSpriteRendererProperties, &NiagaraLightRendererProperties]}
  {emitterName=Flame001, rendererClasses [&NiagaraSpriteRendererProperties]}
 ]
 userVariables [
  User.bisSwinging {type=NiagaraBool}
  User.Skeletal Mesh {
   type=NiagaraDataInterfaceSkeletalMesh
   defaultValue {
    struct=NiagaraExt_VariableValue_DataInterface
    value {dataInterface=&self:NiagaraDataInterfaceSkeletalMesh_0}
   }
  }
 ]
}
