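# Niagara swing trail for the Dyed Duelist pickaxe, built from three short-lived ribbon emitters using the shared AnimTrail spawn, init and update modules.
# Fire and Smoke sample mesh sockets, launch particles along the socket rotation and fade from violet to dark over life, with time-of-day alpha scaling.
# Fire adds strong curl noise and widens, while Smoke narrows out.
# Wash is a simpler fading white-to-grey ribbon.
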NiagaraSystem /BRCosmetics/Weapons/FORT_Melee/Pickaxe_DyedDuelist/FX/NS_Pickaxe_DyedDuelist_Trail
dependencies {
 usedDataInterfaces [&NiagaraDataInterfaceColorCurve, &NiagaraDataInterfaceCurve]
 usedDynamicInputs []
 usedModules [&SystemState]
}
emitters {
 Trail_Fire {
  emitterUpdateScript [
   &EmitterState(Life Cycle Mode=::NewEnumerator1, Loop Behavior=::NewEnumerator1, Loop Duration=0.17, UseLoopDelay=true)
   &FN_SpawnAnimTrail_v2(AnimTrailWidthMode=::FromFirst)
  ]
  particleSpawnScript [
   &SampleSkeletalMesh(
    AdditionalYawPitchRoll=false
    Mesh Sampling Type:ENiagaraSkelSamplingModeFull=::NewEnumerator5
    Skeletal Mesh:NiagaraDataInterfaceSkeletalMesh=$User.SkeletalMesh
    Socket Sampling Mode:ENiagaraSocketSamplingMode=::NewEnumerator5
   )
   /Game/Effects/Niagara/Modules/AnimTrail/FN_InitializeAnimTrail/FN_InitializeAnimTrailBoneIndecies(Second Bone or Socket Index=1)
   /Game/Effects/Niagara/Modules/AnimTrail/FN_InitializeAnimTrail/FN_InitializeAnimTrail(
    AnimTrailWidth=0.65
    Lifetime=0.2
    Position Offset:Vector3f=&Multiply_VectorByFloat(
     Vector=&NormalizeVector(
      Vector To Normalize=$Particles.SampleSkeletalMesh.SampledVelocity
      Safe Normalize=true
      Safe Fallback Threshold=0
      Safe Fallback Vector=<0 0 0.001>
     )
     Float=0
    )
   ) {module=FN_InitializeAnimTrail001}
   &AddVelocity(
    Rotation Coordinate Space=::Simulation
    Velocity=&RotateVectorByQuaternion(Vector To Rotate=<1 0 0>, Quaternion=$Particles.SampleSkeletalMesh.SampledRotation)
    Velocity Speed Scale=100
   ) {module=AddVelocity001}
  ]
  particleUpdateScript [
   &FN_UpdateAnimTrail_v2()
   &ScaleRibbonWidth(Ribbon Width Scale=&FloatFromCurve(FloatCurve=DataInterface() {propertyValues {Curve {keys [
         {interpMode=RCIM_Cubic, tangentMode=RCTM_User, time=0, value=0.5}
         {arriveTangent=1.3038, interpMode=RCIM_Cubic, leaveTangent=1.3038, tangentMode=RCTM_User, time=1, value=1}
        ]}}})) {module=ScaleRibbonWidth001}
   &CurlNoiseForce(Noise Frequency=22, Noise Strength=666, Pan Noise=false, Random Seed=33)
   &GravityForce(Gravity=<0 0 0>)
   &SolveForcesAndVelocity(Clamp Velocity=false, Limit Acceleration=false)
   &Color(
    Color=&ColorFromCurve(ColorCurve=DataInterface() {propertyValues {
       AlphaCurve {keys [
         {arriveTangent=-3.2415, interpMode=RCIM_Cubic, leaveTangent=-3.2415, tangentMode=RCTM_User, time=0.0037, value=1.5}
         {arriveTangentWeight=0.326, interpMode=RCIM_Cubic, tangentMode=RCTM_User, time=0.5608, value=0}
        ]}
       BlueCurve {keys [
         <at 0.0149 1>
         {arriveTangent=-6.9405, interpMode=RCIM_Cubic, leaveTangent=-6.9405, time=0.0689, value=1}
         {arriveTangent=-7.6196, interpMode=RCIM_Cubic, leaveTangent=-7.6196, time=0.123, value=0.25}
         <at 0.1882 0.0914>
        ]}
       GreenCurve {keys [<at 0.0149 0>,<at 0.0689 0 RCIM_Cubic>,<at 0.123 0 RCIM_Cubic>,<at 0.1882 0>]}
       RedCurve {keys [
         <at 0.0149 0.6458>
         {arriveTangent=-5.278, interpMode=RCIM_Cubic, leaveTangent=-5.278, time=0.0689, value=0.3035}
         {arriveTangent=-2.3176, interpMode=RCIM_Cubic, leaveTangent=-2.3176, time=0.123, value=0.0755}
         <at 0.1882 0.0271>
        ]}
      }})
    Scale Alpha=&FNTimeOfDayScalar(Night Scale=0.05, Morning Scale=0.5, Day Scale=0.5, Evening Scale=0.1)
   ) {module=Color001}
   &DynamicMaterialParameters(
    Index 0 Param 3=&FloatFromCurve(FloatCurve=DataInterface() {propertyValues {Curve {keys [
         {arriveTangent=-1.8694, interpMode=RCIM_Cubic, leaveTangent=-1.8694, tangentMode=RCTM_User, time=0, value=1}
         {interpMode=RCIM_Cubic, tangentMode=RCTM_User, time=1, value=0}
        ]}}})
    Module.Index0Param0Write=false
    Module.Index0Param1Write=false
    Module.Index0Param2Write=true
    Module.Index0Param3Write=false
   ) {module=DynamicMaterialParameters001}
  ]
  rendererClasses [&NiagaraRibbonRendererProperties]
  renderers [{rendererClass=&NiagaraRibbonRendererProperties}]
 }
 Trail_Fire_Smoke {
  emitterUpdateScript [
   &EmitterState(Life Cycle Mode=::NewEnumerator1, Loop Behavior=::NewEnumerator1, Loop Duration=0.27, UseLoopDelay=true)
   &FN_SpawnAnimTrail_v2(AnimTrailWidthMode=::FromFirst)
  ]
  particleSpawnScript [
   &SampleSkeletalMesh(
    AdditionalYawPitchRoll=false
    Bone / Socket Index:NiagaraInt32=&Add_Integer(A=$Emitter.SpawnGroup, B=0)
    Mesh Sampling Type:ENiagaraSkelSamplingModeFull=::NewEnumerator5
    Skeletal Mesh:NiagaraDataInterfaceSkeletalMesh=$User.SkeletalMesh
    Socket Sampling Mode:ENiagaraSocketSamplingMode=::NewEnumerator5
   )
   &FN_InitializeAnimTrailBoneIndecies(First Bone or Socket Index=1)
   &FN_InitializeAnimTrail(
    AnimTrailWidth=1
    Color=<rgba 0 0 0 4>
    Lifetime=0.3
    Position Offset:Vector3f=&Multiply_VectorByFloat(
     Vector=&NormalizeVector(
      Vector To Normalize=$Particles.SampleSkeletalMesh.SampledVelocity
      Safe Normalize=true
      Safe Fallback Threshold=0
      Safe Fallback Vector=<0 0 0.001>
     )
     Float=15
    )
   )
   &SystemLocation(MaskSpawnGroup=false)
   &AddVelocity(
    Rotation Coordinate Space=::Simulation
    Velocity=&RotateVectorByQuaternion(Vector To Rotate=<1 0 2>, Quaternion=$Particles.SampleSkeletalMesh.SampledRotation)
    Velocity Speed Scale=100
   ) {module=AddVelocity001}
  ]
  particleUpdateScript [
   &FN_UpdateAnimTrail_v2()
   &ScaleRibbonWidth(Ribbon Width Scale=&FloatFromCurve(FloatCurve=DataInterface() {propertyValues {Curve {keys [
         {tangentMode=RCTM_User, time=0, value=1}
         {arriveTangent=-0.0405, interpMode=RCIM_Cubic, leaveTangent=-0.0405, tangentMode=RCTM_User, time=0.7009, value=1}
         {arriveTangent=-8.0089, interpMode=RCIM_Cubic, leaveTangent=-8.0089, tangentMode=RCTM_User, time=1, value=0}
        ]}}})) {module=ScaleRibbonWidth001}
   &GravityForce(Gravity=<0 0 0>)
   &SolveForcesAndVelocity(Clamp Velocity=false, Limit Acceleration=false)
   &DynamicMaterialParameters(
    Index 0 Param 1=&RandomRangeFloat(Minimum=-0.25, Maximum=0.25)
    Index 0 Param 2=&FloatFromCurve(FloatCurve=DataInterface() {propertyValues {Curve {keys [<at 0 1>,<at 1 0.55>]}}})
    Module.Param0WriteEnabled=true
    Module.Param1WriteEnabled=true
    Module.Param2WriteEnabled=false
    Module.Param3WriteEnabled=false
   )
   &Color(
    Color=&ColorFromCurve(ColorCurve=DataInterface() {propertyValues {
       AlphaCurve {keys [
         {arriveTangent=-3.2415, interpMode=RCIM_Cubic, leaveTangent=-3.2415, tangentMode=RCTM_User, time=0.0649, value=1.5}
         {arriveTangentWeight=0.326, interpMode=RCIM_Cubic, tangentMode=RCTM_User, time=1, value=0}
        ]}
       BlueCurve {keys [
         <at 0 3>
         {arriveTangent=-11.0834, interpMode=RCIM_Cubic, leaveTangent=-11.0834, time=0.1001, value=3}
         {arriveTangent=-7.6311, interpMode=RCIM_Cubic, leaveTangent=-7.6311, time=0.2481, value=0.25}
         <at 0.492 0.0091>
        ]}
       GreenCurve {keys [
         <at 0 0>
         <at 0.1001 0 RCIM_Cubic>
         {arriveTangent=0.0165, interpMode=RCIM_Cubic, leaveTangent=0.0165, time=0.2481, value=0}
         <at 0.492 0.0065>
        ]}
       RedCurve {keys [
         <at 0 1.9375>
         {arriveTangent=-7.5046, interpMode=RCIM_Cubic, leaveTangent=-7.5046, time=0.1001, value=0.9105}
         {arriveTangent=-2.2832, interpMode=RCIM_Cubic, leaveTangent=-2.2832, time=0.2481, value=0.0755}
         <at 0.492 0.0156>
        ]}
      }})
    Scale Alpha=&FNTimeOfDayScalar(Night Scale=0.5, Morning Scale=0.5, Day Scale=0.5, Evening Scale=0.5)
   )
  ]
  rendererClasses [&NiagaraRibbonRendererProperties]
  renderers [{rendererClass=&NiagaraRibbonRendererProperties}]
 }
 Trail_Wash {
  emitterUpdateScript [
   &EmitterState(Life Cycle Mode=::NewEnumerator1, Loop Behavior=::NewEnumerator1, Loop Delay=0.05, Loop Duration=0.2, UseLoopDelay=false)
   &FN_SpawnAnimTrail_v2()
  ]
  particleSpawnScript [
   &FN_InitializeAnimTrailBoneIndecies(First Bone or Socket Index=1)
   &FN_InitializeAnimTrail(AnimTrailWidth=1, Lifetime=0.25)
  ]
  particleUpdateScript [
   &FN_UpdateAnimTrail_v2()
   &ScaleColor(
    Scale Alpha=&FloatFromCurve(
     FloatCurve=DataInterface() {propertyValues {Curve {keys [{arriveTangent=-2.1531, interpMode=RCIM_Cubic, leaveTangent=-2.1531, tangentMode=RCTM_User, time=0, value=1}, <at 1 0 RCIM_Cubic>]}}}
     Scale Curve=0.4
    )
    Scale RGB=&Multiply_Vector(
     A=<1 1 1>
     B=&MakeVectorFromLinearColorRGB(Color=&ColorFromCurve(ColorCurve=DataInterface() {propertyValues {
         AlphaCurve {keys [<at 0 1 RCIM_Cubic>]}
         BlueCurve {keys [<at 0 1 RCIM_Cubic>,<at 0.0639 1>,<at 0.5577 0.3125 RCIM_Cubic>]}
         GreenCurve {keys [<at 0 1 RCIM_Cubic>,<at 0.0639 0.8805>,<at 0.5577 0.2004 RCIM_Cubic>]}
         RedCurve {keys [<at 0 1 RCIM_Cubic>,<at 0.0639 0.8785>,<at 0.5577 0.2082 RCIM_Cubic>]}
        }}))
    )
    ScaleA=true
    ScaleRGB=true
   )
  ]
  rendererClasses [&NiagaraRibbonRendererProperties]
  renderers [{rendererClass=&NiagaraRibbonRendererProperties}]
 }
}
summary {
 systemName=NS_Pickaxe_DyedDuelist_Trail
 emitters [
  {emitterName=Trail_Wash, rendererClasses [&NiagaraRibbonRendererProperties]}
  {emitterName=Trail_Fire, rendererClasses [&NiagaraRibbonRendererProperties]}
  {emitterName=Trail_Fire_Smoke, rendererClasses [&NiagaraRibbonRendererProperties]}
 ]
 userVariables [
  User.bIsFrontEndPreview {type=NiagaraBool}
  User.RibbonIDOffset {type=NiagaraInt32}
  User.SkeletalMesh {
   type=NiagaraDataInterfaceSkeletalMesh
   defaultValue {
    struct=NiagaraExt_VariableValue_DataInterface
    value {dataInterface=&self:NiagaraDataInterfaceSkeletalMesh_0}
   }
  }
  User.TrailColorIndex {type=NiagaraFloat, defaultValue=0, description=""}
 ]
}
