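# Material function for the Dyed Duelist pickaxe that blends a stylized crystal look over incoming material attributes by an Alpha input.
# Base color uses a bump-offset interior texture with depth-map parallax and Fresnel tint.
# Emissive layers combine sky-light reflection, sun-direction backlighting and an animated spell effect.
# The spell uses radial panning around a pre-skinned-position pivot with noise distortion, exposing SpellSpeed, ColorProgress and Reactive inputs, plus a UV debug switch.
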MaterialFunction /BRCosmetics/Weapons/FORT_Melee/Pickaxe_DyedDuelist/Materials/MF_DyedDuelist_Crystal
graph {nodes [
  n1 = NamedRerouteUsage()
  n2 = NamedRerouteUsage()
  n3 = NamedRerouteUsage()
  n4 = NamedRerouteUsage()
  n5 = Multiply(
   A=VectorParameter() {group=DDCrystal, parameterName=DD_BaseColorMultiply, sortPriority=0, defaultValue=<rgba 1 1 1 1>}.RGB
   B=NamedRerouteUsage()
  )
  n6 = TextureSampleParameter2D(UVs=NamedRerouteUsage()) {
   group=DDCrystal
   parameterName=DD_InteriorTexture
   sortPriority=5
   texture=/Game/Accessories/FORT_Backpacks/Backpack_F_MED_RoyalDusk/Textures/T_RoyalDusk_RoyalDusk_Crystal1
  }
  n7 = OneMinus(None=ScalarParameter() {defaultValue=0.5, group=DDCrystal, parameterName=DD_InteriorPrismColors, sortPriority=2})
  n8 = Multiply(
   A=Multiply(A=@n5, B=Desaturation(None=@n6.RGB, Fraction=@n7))
   B=ScalarParameter() {defaultValue=12, group=DDCrystal, parameterName=DD_InteriorIntensity, sortPriority=1}
  ) {constB=0.1}
  n9 = ScalarParameter() {defaultValue=0.6, group=DDCrystal, parameterName=DD_InteriorContrast, sortPriority=3}
  n10 = VectorParameter() {group=DDCrystal, parameterName=DD_FresnelColor, sortPriority=19, defaultValue=<rgba 0.7823 0.3333 1 1>}
  n11 = NamedRerouteUsage()
  n12 = NamedRerouteUsage()
  n13 = LinearInterpolate(
   A=NamedRerouteDeclaration(None=LinearInterpolate(A=@n5, B=@n8, Alpha=@n9)) {name=InteriorColor, nodeColor=<rgba 1 0 0.4688 1>}
   B=@n10.RGB
   Alpha=Multiply(A=Multiply(A=@n11, B=@n12), B=NamedRerouteUsage())
  )
  n14 = Add(A=NamedRerouteUsage(), B=NamedRerouteUsage())
  n15 = NamedRerouteUsage()
  n16 = NamedRerouteUsage()
  n17 = NamedRerouteUsage()
  n18 = Add(A=LinearInterpolate(A=@n14, B=Add(A=@n14, B=@n15), Alpha=Multiply(A=@n16, B=@n17)) {constAlpha=0}, B=NamedRerouteUsage())
  n19 = FunctionInput() {bUsePreviewValueAsDefault=true, inputName=Alpha, inputType=FunctionInput_Scalar, sortPriority=1, previewValue=<1 1 1 1>}
  FunctionOutput(None=BlendMaterialAttributes(
    A=SetMaterialAttributes!(MaterialAttributes=@n1, BaseColor=@n2, EmissiveColor=@n3) {attributeSetTypes [#1, #2]}
    B=SetMaterialAttributes(MaterialAttributes=@n4, BaseColor=@n13, EmissiveColor=@n18) {attributeSetTypes [#1, #2]}
    Alpha=@n19
   ))
  Multiply(A=Time()) {constB=0.2}
  n20 = GetMaterialAttributes(None=FunctionInput() {inputName=Material, inputType=FunctionInput_MaterialAttributes}) {attributeGetTypes [#1, #2]}
  NamedRerouteDeclaration(None=@n20.EmissiveColor) {name=Emissive, nodeColor=<rgba 1 0 0.75 1>}
  NamedRerouteDeclaration(None=@n20.MaterialAttributes) {name=MA, nodeColor=<rgba 0.3406 1 0.0156 1>}
  NamedRerouteDeclaration(None=@n20.BaseColor) {name=BaseColor, nodeColor=<rgba 0.1661 0.3253 1 1>}
  n21 = TextureCoordinate()
  n22 = NamedRerouteDeclaration(None=@n21) {name=UVCoord, nodeColor=<rgba 0.5938 1 0 1>}
  n23 = ScalarParameter() {defaultValue=0.4, group=DDCrystal, parameterName=DD_InteriorUVTiling, sortPriority=4}
  n24 = ScalarParameter() {group=DDCrystal, parameterName=DD_DepthMinimum, sortPriority=6}
  n25 = ScalarParameter() {defaultValue=-10, group=DDCrystal, parameterName=DD_DepthMaximum, sortPriority=7}
  n26 = TextureSampleParameter2D(UVs=Multiply(A=@n21, B=ScalarParameter() {defaultValue=2.6, group=DDCrystal, parameterName=DD_DepthUVTiling, sortPriority=8})) {
   group=DDCrystal
   parameterName=DD_DepthTexture
   samplerType=SAMPLERTYPE_LinearGrayscale
   sortPriority=9
   texture=/Game/Accessories/FORT_Backpacks/Backpack_F_MED_RoyalDusk/Textures/T_RoyalDusk_RoyalDusk_Crystal2
  }
  n27 = NamedRerouteDeclaration(None=@n26.R) {name=SparksDepth, nodeColor=<rgba 0.1156 1 0.3086 1>}
  NamedRerouteDeclaration(None=BumpOffset(Coordinate=Multiply(A=@n22, B=@n23), Height=LinearInterpolate(A=@n24, B=@n25, Alpha=@n27) {constA=-4, constB=-20})) {name=UVOffsetCoord, nodeColor=<rgba 0.9453 0 1 1>}
  n28 = PixelNormalWS()
  n29 = CameraVectorWS()
  n30 = OneMinus(None=Saturate(None=DotProduct(A=@n28, B=@n29)))
  NamedRerouteDeclaration(None=Multiply(A=Multiply(A=@n30, B=@n30), B=ScalarParameter() {defaultValue=0.2, group=DDCrystal, parameterName=DD_FresnelIntensity, sortPriority=20})) {name=Fresnel, nodeColor=<rgba 0 1 0.1328 1>}
  n31 = TextureSampleParameter2D(UVs=NamedRerouteUsage()) {
   group=DDCrystal
   parameterName=DD_BacklightNormal
   samplerType=SAMPLERTYPE_Normal
   sortPriority=18
   texture=/Game/Accessories/FORT_Backpacks/Backpack_F_MED_RoyalDusk/Textures/T_RoyalDusk_RoyalDusk_CrystalNormal1
  }
  n32 = Transform(None=LinearInterpolate(
    A=Constant3Vector() {constant=<rgba 0 0 1 0>}
    B=@n31.RGB
    Alpha=ScalarParameter() {defaultValue=0.2, group=DDCrystal, parameterName=DD_BacklightNormalIntensity, sortPriority=15}
   ) {constAlpha=0.1})
  n33 = CameraVectorWS()
  n34 = Multiply(A=ScalarParameter() {defaultValue=1, group=DDCrystal, parameterName=DD_BacklightViewAngle, sortPriority=16}, B=@n33)
  n35 = DotProduct(A=Normalize(VectorInput=Add(A=@n32, B=@n34)), B=Multiply(A=SkyAtmosphereLightDirection()) {constB=-1})
  n36 = Saturate(None=Add(A=Multiply(A=@n35) {constB=0.5}) {constB=0.5})
  n37 = ScalarParameter() {defaultValue=2, group=DDCrystal, parameterName=DD_BacklightAperture, sortPriority=17}
  n38 = DotProduct(A=@n32, B=@n33)
  n39 = Multiply(A=Multiply(A=Power(Base=@n36, Exp=@n37), B=Saturate(None=@n38)), B=NamedRerouteUsage())
  n40 = VectorParameter() {group=DDCrystal, parameterName=DD_BacklightColorMultiply, sortPriority=13, defaultValue=<rgba 1 1 1 1>}
  NamedRerouteDeclaration(None=Multiply(A=Multiply(A=@n39, B=@n40.RGB), B=ScalarParameter() {defaultValue=2, group=DDCrystal, parameterName=DD_BackLightIntensity, sortPriority=14})) {name=Backlight, nodeColor=<rgba 1 0 0.2578 1>}
  n41 = ReflectionVectorWS()
  n42 = ScalarParameter() {defaultValue=0.1, group=DDCrystal, parameterName=DD_ReflectionRoughness, sortPriority=22}
  n43 = VectorParameter() {group=DDCrystal, parameterName=DD_ReflectionColor, sortPriority=21, defaultValue=<rgba 0.0952 0.0364 0.1146 1>}
  NamedRerouteDeclaration(None=Multiply(A=SkyLightEnvMapSample(Direction=@n41, Roughness=@n42), B=@n43.RGB)) {name=Reflection, nodeColor=<rgba 1 0.1406 0 1>}
  n44 = AppendVector(
   A=MaterialFunctionCall!(UV=NamedRerouteUsage(), UV Tilling, PannerValue) {materialFunction=&/MF_RadialUVPanner}.UV
   B=Constant()
  )
  n45 = Constant3Vector() {constant=<rgba 0.6927 0.1078 0.6051 1>}
  n46 = NamedRerouteUsage()
  n47 = NamedRerouteUsage()
  n48 = Constant3Vector() {constant=<rgba 0.0358 0.3229 0 1>}
  n49 = Multiply(A=LinearInterpolate(A=Multiply(A=@n45, B=@n46), B=Multiply(A=@n47, B=@n48), Alpha=NamedRerouteUsage())) {constB=2}
  n50 = NamedRerouteUsage()
  n51 = OneMinus(None=Distance(A=Multiply(A=@n50) {constB=0.8}, B=Constant() {r=0.4}))
  NamedRerouteDeclaration(None=LinearInterpolate(B=StaticSwitchParameter(True=@n44, False=@n49) {group=DDCrystal, parameterName=DD_UVPivotDebug}, Alpha=Step(X=@n51))) {name=CrystalEffect, nodeColor=<rgba 0.4766 1 0 1>}
  NamedRerouteDeclaration(None=FunctionInput() {inputName=ColorProgress, inputType=FunctionInput_Scalar, sortPriority=2}) {name=ColorProgress, nodeColor=<rgba 0 1 0.1328 1>}
  n52 = VectorParameter() {group=DDCrystal, parameterName=DD_PivotUVPos, defaultValue=<rgba 0 -8 -58 0>}
  n53 = Divide(
   A=Add(A=PreSkinnedPosition(), B=ComponentMask(None=@n52.RGB) {b=true, g=true, r=true})
   B=ScalarParameter() {defaultValue=4.65, group=DDCrystal, parameterName=DD_UVDivision}
  ) {constB=4.65}
  NamedRerouteDeclaration(None=ComponentMask(None=VertexInterpolator(VS=@n53).PS) {b=true, g=true}) {name=PivotUV, nodeColor=<rgba 1 0 0.7266 1>}
  n54 = Multiply(A=Time()) {constB=0.5}
  n55 = NamedRerouteUsage()
  n56 = MaterialFunctionCall(UV=NamedRerouteUsage(), PannerValue=AppendVector(A=Constant(), B=Add(A=@n54, B=@n55))) {materialFunction=&MF_RadialUVPanner}
  n57 = Time()
  n58 = TextureSample(UVs=Add(A=TextureCoordinate() {uTiling=10, vTiling=10}, B=Multiply(A=@n57) {constB=0.5})) {samplerType=SAMPLERTYPE_LinearColor, texture=/BRCosmetics/Weapons/FORT_Melee/Pickaxe_DyedDuelist/Textures/T_DyedDuelist_Noise}
  n59 = TextureSample!(
   UVs=Add(A=Add(A=@n56.UV) {constB=-0.2}, B=Multiply(A=@n58.R) {constB=0.2})
   Tex
   "DDX(UVs)"=@n56.DDX
   "DDY(UVs)"=@n56.DDX
   Apply View MipBias
  ) {
   mipValueMode=TMVM_Derivative
   samplerType=SAMPLERTYPE_LinearColor
   texture=/BRCosmetics/Weapons/FORT_Melee/Pickaxe_DyedDuelist/Textures/T_DyedDuelist_FX
  }
  NamedRerouteDeclaration(None=@n59.G) {name=SpellTexG, nodeColor=<rgba 0 1 0 1>}
  NamedRerouteDeclaration(None=@n59.R) {name=SpellTexR, nodeColor=<rgba 1 0 0.75 1>}
  NamedRerouteDeclaration(None=FunctionInput() {inputName=SpellSpeed, inputType=FunctionInput_Scalar, sortPriority=3}) {name=SpellSpeed, nodeColor=<rgba 1 0.289 0.361 1>}
  NamedRerouteDeclaration(None=FunctionInput() {inputName=Reactive, inputType=FunctionInput_Scalar, sortPriority=2}) {name=Reactive, nodeColor=<rgba 0 1 0.1328 1>}
  Reroute(None=@n26.R)
  Reroute(None=@n59.G)
  Reroute(None=@n59.R)
 ]}
