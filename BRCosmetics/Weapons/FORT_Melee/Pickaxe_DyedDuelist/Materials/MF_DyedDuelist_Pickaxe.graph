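# Material function for the Dyed Duelist pickaxe that adds a colored, animated Fresnel glow to incoming material attributes, apparently for the handle.
# Two fire-noise samples scroll in opposite directions and combine with Max to form the glow texture.
# Fresnel terms combine with this noise and the TAMask, VibrantEmissiveAmount and Reactive inputs.
# Time of day scales intensity, lower in evening and night, before the result is added to Emissive.
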MaterialFunction /BRCosmetics/Weapons/FORT_Melee/Pickaxe_DyedDuelist/Materials/MF_DyedDuelist_Pickaxe
graph {nodes [
  Constant() {r=1}
  n1 = NamedRerouteUsage()
  n2 = NamedRerouteUsage()
  n3 = Fresnel() {baseReflectFraction=0.06, exponent=2}
  n4 = Multiply(A=@n3) {constB=5}
  n5 = NamedRerouteUsage()
  n6 = Multiply(A=Add(A=Multiply(A=@n2, B=@n4), B=Multiply(A=@n3) {constB=2}), B=Add(A=OneMinus(None=@n5), B=Multiply(A=@n5) {constB=0.1}))
  n7 = ScalarParameter() {defaultValue=5, group=DDMaterial, parameterName=DD_FresnelHandleIntensity}
  n8 = Divide(A=@n7)
  n9 = MaterialFunctionCall(Value_Morning=@n7, Value_Day=@n7, Value_Evening=@n8, Value_Night=@n8) {materialFunction=&Scalar_Time_of_Day}
  n10 = VectorParameter() {group=DDMaterial, parameterName=DD_FresnelHandleColor, sortPriority=33, defaultValue=<rgba 0.6927 0.1078 0.6051 1>}
  n11 = Multiply(A=Multiply(A=Multiply(A=@n6, B=@n9.Result), B=@n10.RGB), B=Multiply(A=NamedRerouteUsage(), B=NamedRerouteUsage()))
  FunctionOutput(None=SetMaterialAttributes!(MaterialAttributes=NamedRerouteUsage(), EmissiveColor=Add(A=@n1, B=@n11)) {attributeSetTypes [#1]})
  Multiply(A=@n3, B=@n3)
  n12 = GetMaterialAttributes(None=FunctionInput() {inputType=FunctionInput_MaterialAttributes}) {attributeGetTypes [#1]}
  NamedRerouteDeclaration(None=@n12.MaterialAttributes) {name=MA, nodeColor=<rgba 0 0.8594 1 1>}
  NamedRerouteDeclaration(None=@n12.EmissiveColor) {name=InEmissive, nodeColor=<rgba 1 0.7734 0 1>}
  NamedRerouteDeclaration(None=FunctionInput() {inputName=TAMask, inputType=FunctionInput_Scalar, sortPriority=1}) {name=TAMask, nodeColor=<rgba 0 0.5547 1 1>}
  NamedRerouteDeclaration(None=FunctionInput() {inputName=VibrantEmissiveAmount, inputType=FunctionInput_Scalar, sortPriority=2}) {name=VibrantEmissiveAmount, nodeColor=<rgba 1 0 0.75 1>}
  n13 = Multiply(A=TextureCoordinate() {uTiling=0.8, vTiling=0.8}) {constB=2}
  n14 = Time()
  n15 = Multiply(A=Constant2Vector() {g=1}) {constB=0.1}
  n16 = Add(A=@n13, B=Multiply(A=@n14, B=@n15))
  n17 = Multiply(A=@n15) {constB=-1}
  n18 = Multiply(A=Add(A=@n13, B=Multiply(A=@n14, B=@n17))) {constB=1.1}
  NamedRerouteDeclaration(None=Max(
    A=TextureSample(UVs=@n16) {samplerSource=SSM_Wrap_WorldGroupSettings, texture=/Game/Global/Textures/Default/Noise/T_FireNoise_01}.R
    B=TextureSample(UVs=@n18) {samplerSource=SSM_Wrap_WorldGroupSettings, texture=&T_FireNoise_01}.R
   )) {name=TexFX, nodeColor=<rgba 1 0 0.2344 1>}
  NamedRerouteDeclaration(None=FunctionInput() {inputName=Reactive, inputType=FunctionInput_Scalar, sortPriority=2}) {name=Reactive, nodeColor=<rgba 1 0 0.75 1>}
 ]}
