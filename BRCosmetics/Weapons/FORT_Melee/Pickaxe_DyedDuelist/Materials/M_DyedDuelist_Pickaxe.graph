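# Character-style material for the Dyed Duelist pickaxe, using material attributes with shading model chosen by expression.
# MF_CharacterParent_2023_PreFX feeds the crystal function, masked by the Tech Art Mask's blue channel, then the pickaxe function and MF_CharacterParent_2023_PostFX.
# The pickaxe function receives a combined mask of doubled red plus blue.
# A shared Reactive parameter drives both Dyed Duelist functions, while named reroutes supply color progress, spell speed and vibrant emissive inputs.
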Material /BRCosmetics/Weapons/FORT_Melee/Pickaxe_DyedDuelist/Materials/M_DyedDuelist_Pickaxe
graph {
 properties {
  bUseMaterialAttributes=true
  bUsedWithMeshDeformer=true
  bUsedWithNanite=true
  bUsedWithSkeletalMesh=true
  pixelDepthOffsetMode=PDOM_Legacy
  shadingModel=MSM_FromMaterialExpression
  usedShadingModels=Default Lit | Subsurface | Clear Coat | Subsurface Profile | Cloth
 }
 nodes [
  n1 = MaterialFunctionCall() {materialFunction=&MF_CharacterParent_2023_PreFX}
  n2 = ComponentMask(None=@n1.Tech Art Mask) {b=true}
  n3 = ScalarParameter() {group=DDMaterial, parameterName=Reactive}
  n4 = NamedRerouteUsage()
  n5 = NamedRerouteUsage()
  n6 = MaterialFunctionCall(Float3=@n1.Tech Art Mask) {materialFunction=&BreakOutFloat3Components}
  n7 = Add(A=@n6.R, B=@n6.R)
  MaterialFunctionCall(
   Material=MaterialFunctionCall!(
    In=MaterialFunctionCall!(Material=@n1.Result, Alpha=@n2, Reactive=@n3, ColorProgress=@n4, SpellSpeed=@n5) {materialFunction=&/MF_DyedDuelist_Crystal}.Result
    TAMask=Add(A=@n7, B=@n6.B)
    Reactive=@n3
    VibrantEmissiveAmount=NamedRerouteUsage()
   ) {materialFunction=&/MF_DyedDuelist_Pickaxe}.Result
   M=@n1.M
   Pre Skinned Local Position=@n1.M
   Distance Blend=@n1.M
   Fresnel=@n1.M
  ) {materialFunction=&MF_CharacterParent_2023_PostFX}
  NamedRerouteDeclaration(None=ScalarParameter() {group=DDMaterial, parameterName=VibrantEmissiveAmount}) {name=VibrantEmissiveAmount, nodeColor=<rgba 0.2188 1 0 1>}
  NamedRerouteDeclaration(None=ScalarParameter() {group=DDMaterial, parameterName=SpellSpeed}) {name=SpellSpeed, nodeColor=<rgba 1 0.6328 0 1>}
 ]
}
