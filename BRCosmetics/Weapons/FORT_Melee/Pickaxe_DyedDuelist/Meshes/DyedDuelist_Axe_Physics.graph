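# Physics asset for the Dyed Duelist two-handed pickaxe cosmetic, providing collision and joint limits for its dangling bow bones.
# Two capsules cover the handle, which anchors four chains of small capsule bodies: left and right, upper and bottom bows.
# Each chain links bone to bone with swing and twist limits, tighter on the two-segment upper bows than on the three-segment bottom ones.
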PhysicsAsset /BRCosmetics/Weapons/FORT_Melee/Pickaxe_DyedDuelist/Meshes/DyedDuelist_Axe_Physics
bodies {
 Melee_TwoHanded_Handle [
  {
   shapeType=Capsule
   length=82.4688
   radius=2.6896
   shapeName=Melee_TwoHanded_Handle_capsule
   center=<0.7595 -0.9943 25.5824>
   rotation=<rot -0.0048 -90 0.0274>
  }
  {shapeType=Capsule, length=6.0096, radius=4.5346, shapeName=Melee_TwoHanded_Handle_capsule, center=<0 -0.8466 40.3787>}
 ]
 dyn_L_bottom_bow_01 [{shapeType=Capsule, length=3, radius=2, shapeName=dyn_L_bottom_bow_01_capsule, center=<3.4687 0.2162 0.077>, rotation=<rot 90 0 0>}]
 dyn_L_bottom_bow_02 [{shapeType=Capsule, length=3, radius=2, shapeName=dyn_L_bottom_bow_02_capsule, center=<3.6258 0.0684 -0.2722>, rotation=<rot 90 0 0>}]
 dyn_L_bottom_bow_03 [{shapeType=Capsule, length=3, radius=2, shapeName=dyn_L_bottom_bow_03_capsule, center=<5.6592 -0.2335 -0.8374>, rotation=<rot 90 0 0>}]
 dyn_L_upper_bow_01 [{shapeType=Capsule, length=3, radius=2, shapeName=dyn_L_upper_bow_01_capsule, center=<1.7271 0.9554 -0.72>, rotation=<rot 90 0 0>}]
 dyn_L_upper_bow_02 [{shapeType=Capsule, length=3, radius=2, shapeName=dyn_L_upper_bow_02_capsule, center=<1.9261 1.1874 -0.1478>, rotation=<rot 90 0 0>}]
 dyn_R_bottom_bow_01 [{shapeType=Capsule, length=3, radius=2, shapeName=dyn_R_bottom_bow_01_capsule, center=<-1.9495 -0.0736 0.2651>, rotation=<rot 90 0 0>}]
 dyn_R_bottom_bow_02 [{shapeType=Capsule, length=3, radius=2, shapeName=dyn_R_bottom_bow_02_capsule, center=<-2.2294 0.0609 0.5648>, rotation=<rot 90 0 0>}]
 dyn_R_bottom_bow_03 [{shapeType=Capsule, length=3, radius=2, shapeName=dyn_R_bottom_bow_03_capsule, center=<-5.2376 -0.0331 0.4602>, rotation=<rot 90 0 0>}]
 dyn_R_upper_bow_01 [{shapeType=Capsule, length=3, radius=2, shapeName=dyn_R_upper_bow_01_capsule, center=<-1.3468 -1.0404 0.0109>, rotation=<rot 90 0 0>}]
 dyn_R_upper_bow_02 [{shapeType=Capsule, length=3, radius=2, shapeName=dyn_R_upper_bow_02_capsule, center=<-2.4503 -1.1248 0.0963>, rotation=<rot 90 0 0>}]
}
constraints [
 {bone1Name=dyn_L_bottom_bow_01, bone2Name=Melee_TwoHanded_Handle, swing1LimitDegrees=35, swing2LimitDegrees=30, twistLimitDegrees=10}
 {bone1Name=dyn_L_bottom_bow_02, bone2Name=dyn_L_bottom_bow_01, swing1LimitDegrees=35, swing2LimitDegrees=30, twistLimitDegrees=10}
 {bone1Name=dyn_L_bottom_bow_03, bone2Name=dyn_L_bottom_bow_02, swing1LimitDegrees=35, swing2LimitDegrees=30, twistLimitDegrees=10}
 {bone1Name=dyn_L_upper_bow_01, bone2Name=Melee_TwoHanded_Handle, swing1LimitDegrees=20, swing2LimitDegrees=30, twistLimitDegrees=5}
 {bone1Name=dyn_L_upper_bow_02, bone2Name=dyn_L_upper_bow_01, swing1LimitDegrees=20, swing2LimitDegrees=30, twistLimitDegrees=5}
 {bone1Name=dyn_R_bottom_bow_01, bone2Name=Melee_TwoHanded_Handle, swing1LimitDegrees=35, swing2LimitDegrees=30, twistLimitDegrees=10}
 {bone1Name=dyn_R_bottom_bow_02, bone2Name=dyn_R_bottom_bow_01, swing1LimitDegrees=35, swing2LimitDegrees=30, twistLimitDegrees=10}
 {bone1Name=dyn_R_bottom_bow_03, bone2Name=dyn_R_bottom_bow_02, swing1LimitDegrees=35, swing2LimitDegrees=30, twistLimitDegrees=10}
 {bone1Name=dyn_R_upper_bow_01, bone2Name=Melee_TwoHanded_Handle, swing1LimitDegrees=20, swing2LimitDegrees=30, twistLimitDegrees=5}
 {bone1Name=dyn_R_upper_bow_02, bone2Name=dyn_R_upper_bow_01, swing1LimitDegrees=20, swing2LimitDegrees=30, twistLimitDegrees=5}
]
